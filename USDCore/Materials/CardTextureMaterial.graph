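# CardTextureMaterial is a minimal Default Lit material in USDCore, apparently intended for texturing cards.
# A TextureObjectParameter named Texture feeds a single TextureSample, so material instances can swap the image.
# The graph shows no connection from that sample to any material output.
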Material /USDCore/Materials/CardTextureMaterial
graph {
 properties {pixelDepthOffsetMode=PDOM_Legacy, usedShadingModels=Default Lit}
 nodes [TextureSample(Tex=TextureObjectParameter() {parameterName=Texture})]
}
