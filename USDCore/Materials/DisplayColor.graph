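# DisplayColor is a Default Lit USDCore material that, judging by its name, shows USD display colors stored as mesh vertex colors.
# It converts the red, green and blue vertex color channels from sRGB to linear separately, then recombines them with MakeFloat3.
# Vertex alpha is unused, and two scalar constants sit alongside the color chain.
# Usage flags cover geometry caches, instanced static meshes, skeletal meshes, morph targets and mesh deformers.
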Material /USDCore/Materials/DisplayColor
graph {
 properties {
  bUsedWithGeometryCache=true
  bUsedWithInstancedStaticMeshes=true
  bUsedWithMeshDeformer=true
  bUsedWithMorphTargets=true
  bUsedWithSkeletalMesh=true
  pixelDepthOffsetMode=PDOM_Legacy
  usedShadingModels=Default Lit
 }
 nodes [
  Constant()
  Constant() {r=0.5}
  n1 = VertexColor()
  MaterialFunctionCall(
   X=MaterialFunctionCall!(sRGB=@n1.R) {materialFunction=/Engine/Functions/Engine_MaterialFunctions01/ImageAdjustment/sRGBToLinear}.Linear
   Y=MaterialFunctionCall(sRGB=@n1.G) {materialFunction=&sRGBToLinear}.Linear
   Z=MaterialFunctionCall(sRGB=@n1.B) {materialFunction=&sRGBToLinear}.Linear
  ) {materialFunction=&MakeFloat3}
 ]
}
