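# Translucent Default Lit material in USDCore that, going by its name, shows USD display color and opacity stored as vertex colors.
# Red, green and blue each pass through sRGBToLinear and recombine with MakeFloat3, while alpha passes through unconverted.
# Usage flags cover skeletal meshes, morph targets, mesh deformers, instanced static meshes and geometry caches.
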Material /USDCore/Materials/DisplayColorAndOpacity
graph {
 properties {
  bUsedWithGeometryCache=true
  bUsedWithInstancedStaticMeshes=true
  bUsedWithMeshDeformer=true
  bUsedWithMorphTargets=true
  bUsedWithSkeletalMesh=true
  blendMode=BLEND_Translucent
  pixelDepthOffsetMode=PDOM_Legacy
  usedShadingModels=Default Lit
 }
 nodes [
  n1 = VertexColor()
  MaterialFunctionCall(
   X=MaterialFunctionCall!(sRGB=@n1.R) {materialFunction=/Engine/Functions/Engine_MaterialFunctions01/ImageAdjustment/sRGBToLinear}.Linear
   Y=MaterialFunctionCall(sRGB=@n1.G) {materialFunction=&sRGBToLinear}.Linear
   Z=MaterialFunctionCall(sRGB=@n1.B) {materialFunction=&sRGBToLinear}.Linear
  ) {materialFunction=&MakeFloat3}
  Reroute(None=@n1.A)
 ]
}
