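# Material function that applies a USD-style Transform2d (scale, rotation, translation) to texture coordinates.
# It offsets UVs so (0,1) becomes the origin, scales them, rotates them with CustomRotator, adds the translation, then restores the offset.
# Counter-clockwise degrees become negated turns and translation Y is negated, so the transform works from a bottom-left origin in Unreal's flipped-V space.
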MaterialFunction /USDCore/Materials/UsdTransform2d
graph {nodes [
  n1 = Constant2Vector() {g=1}
  n2 = FunctionInput() {inputName=TextureCoordinate, inputType=FunctionInput_Vector2}
  n3 = FunctionInput() {bUsePreviewValueAsDefault=true, inputName=Scale, inputType=FunctionInput_Vector2, previewValue=<1 1 0 1>}
  n4 = FunctionInput() {bUsePreviewValueAsDefault=true, description=Counter-clockwise rotation in degrees, inputName=Rotation, inputType=FunctionInput_Scalar}
  n5 = MaterialFunctionCall(UVs=Multiply(A=Subtract(A=@n2, B=@n1), B=@n3), Rotation Center=Constant2Vector(), "Rotation Angle (0-1)"=Divide(A=@n4) {constB=-360}) {materialFunction=&CustomRotator}
  n6 = Multiply(A=FunctionInput() {bUsePreviewValueAsDefault=true, inputName=Translation, inputType=FunctionInput_Vector2}, B=Constant2Vector() {g=-1, r=1})
  FunctionOutput(None=Add(A=@n1, B=Add(A=@n5.Rotated Values, B=@n6) {constB=0}))
 ]}
