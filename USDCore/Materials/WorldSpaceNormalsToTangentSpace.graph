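# Material function, exposed to the library, that converts a world-space normal input into a normalized tangent-space normal.
# It builds a basis from the vertex world-space tangent, the cross product of tangent and normal, and the vertex world-space normal.
# The engine's InverseTransformMatrix function transforms the input through that basis, and the result is normalized before output.
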MaterialFunction /USDCore/Materials/WorldSpaceNormalsToTangentSpace
graph {
 properties {bExposeToLibrary=true, description="Converts a world space normal to a tangent space normal (also normalizes the output)"}
 nodes [
  n1 = FunctionInput() {previewValue=<0 0 1 1>}
  n2 = VertexTangentWS()
  n3 = VertexNormalWS()
  n4 = CrossProduct(A=@n2, B=@n3)
  FunctionOutput(None=Normalize(VectorInput=MaterialFunctionCall!(VectorToTransform=@n1, BasisX=@n2, BasisY=@n4, BasisZ=@n3) {materialFunction=/Engine/Functions/Engine_MaterialFunctions02/Math/InverseTransformMatrix}.Transformed Vector 3))
 ]
}
