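# UsdPreviewSurface is a parameterized USDCore material modeling USD preview surfaces: base color, metallic, roughness, emissive, opacity, normal, refraction and ambient occlusion.
# Scalar-driven lerps choose constants or textures at runtime, with channel selection, optional sRGB-to-linear vertex color and world-to-tangent normal conversion.
# UVSwitch subgraphs pick each input's UV set, feeding UsdTransform2d.
# Static bUseSwitch toggles Switch nodes or If chains, the safer option for Nanite analytical derivatives.
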Material /USDCore/Materials/UsdPreviewSurface
graph {
 expressions {
  MaterialExpressionComposite_12 {
   class=MaterialExpressionComposite
   inputs [UVIndex {from=MaterialExpressionScalarParameter_16, output=""}]
   properties {
    desc="Used as a replacement for Custom expression, since it doesn't fully translates to analytical derivatives for Nanite"
    subgraphName=UVSwitch
   }
  }
  MaterialExpressionComposite_13 {
   class=MaterialExpressionComposite
   inputs [UVIndex {from=MaterialExpressionScalarParameter_17, output=""}]
   properties {
    desc="Used as a replacement for Custom expression, since it doesn't fully translates to analytical derivatives for Nanite"
    subgraphName=UVSwitch
   }
  }
  MaterialExpressionComposite_14 {
   class=MaterialExpressionComposite
   inputs [UVIndex {from=MaterialExpressionScalarParameter_18, output=""}]
   properties {
    desc="Used as a replacement for Custom expression, since it doesn't fully translates to analytical derivatives for Nanite"
    subgraphName=UVSwitch
   }
  }
  MaterialExpressionComposite_15 {
   class=MaterialExpressionComposite
   inputs [UVIndex {from=MaterialExpressionScalarParameter_19, output=""}]
   properties {
    desc="Used as a replacement for Custom expression, since it doesn't fully translates to analytical derivatives for Nanite"
    subgraphName=UVSwitch
   }
  }
  MaterialExpressionComposite_16 {
   class=MaterialExpressionComposite
   inputs [UVIndex {from=MaterialExpressionScalarParameter_20, output=""}]
   properties {
    desc="Used as a replacement for Custom expression, since it doesn't fully translates to analytical derivatives for Nanite"
    subgraphName=UVSwitch
   }
  }
  MaterialExpressionComposite_17 {
   class=MaterialExpressionComposite
   inputs [UVIndex {from=MaterialExpressionScalarParameter_21, output=""}]
   properties {
    desc="Used as a replacement for Custom expression, since it doesn't fully translates to analytical derivatives for Nanite"
    subgraphName=UVSwitch
   }
  }
  MaterialExpressionComposite_18 {
   class=MaterialExpressionComposite
   inputs [UVIndex {from=MaterialExpressionScalarParameter_22, output=""}]
   properties {
    desc="Used as a replacement for Custom expression, since it doesn't fully translates to analytical derivatives for Nanite"
    subgraphName=UVSwitch
   }
  }
  MaterialExpressionComposite_7 {
   class=MaterialExpressionComposite
   inputs [UVIndex {from=MaterialExpressionScalarParameter_15, output=""}]
   properties {
    desc="Used as a replacement for Custom expression, since it doesn't fully translates to analytical derivatives for Nanite"
    subgraphName=UVSwitch
   }
  }
  MaterialExpressionIf_22 {
   class=MaterialExpressionIf
   inputs [
    A {from=MaterialExpressionReroute_24, output=UVIndex}
    B {from="", output=""}
    "A > B" {from=MaterialExpressionIf_23, output=""}
    "A == B" {from=MaterialExpressionIf_23, output=""}
    "A < B" {from=MaterialExpressionTextureCoordinate_32, output=""}
   ]
   properties {constB=1}
  }
  MaterialExpressionIf_23 {
   class=MaterialExpressionIf
   inputs [
    A {from=MaterialExpressionReroute_24, output=UVIndex}
    B {from="", output=""}
    "A > B" {from=MaterialExpressionIf_24, output=""}
    "A == B" {from=MaterialExpressionIf_24, output=""}
    "A < B" {from=MaterialExpressionTextureCoordinate_33, output=""}
   ]
   properties {constB=2}
  }
  MaterialExpressionIf_24 {
   class=MaterialExpressionIf
   inputs [
    A {from=MaterialExpressionReroute_24, output=UVIndex}
    B {from="", output=""}
    "A > B" {from=MaterialExpressionTextureCoordinate_35, output=""}
    "A == B" {from=MaterialExpressionTextureCoordinate_35, output=""}
    "A < B" {from=MaterialExpressionTextureCoordinate_34, output=""}
   ]
   properties {constB=3}
  }
  MaterialExpressionIf_37 {
   class=MaterialExpressionIf
   inputs [
    A {from=MaterialExpressionReroute_16, output=UVIndex}
    B {from="", output=""}
    "A > B" {from=MaterialExpressionIf_38, output=""}
    "A == B" {from=MaterialExpressionIf_38, output=""}
    "A < B" {from=MaterialExpressionTextureCoordinate_52, output=""}
   ]
   properties {constB=1}
  }
  MaterialExpressionIf_38 {
   class=MaterialExpressionIf
   inputs [
    A {from=MaterialExpressionReroute_16, output=UVIndex}
    B {from="", output=""}
    "A > B" {from=MaterialExpressionIf_39, output=""}
    "A == B" {from=MaterialExpressionIf_39, output=""}
    "A < B" {from=MaterialExpressionTextureCoordinate_53, output=""}
   ]
   properties {constB=2}
  }
  MaterialExpressionIf_39 {
   class=MaterialExpressionIf
   inputs [
    A {from=MaterialExpressionReroute_16, output=UVIndex}
    B {from="", output=""}
    "A > B" {from=MaterialExpressionTextureCoordinate_55, output=""}
    "A == B" {from=MaterialExpressionTextureCoordinate_55, output=""}
    "A < B" {from=MaterialExpressionTextureCoordinate_54, output=""}
   ]
   properties {constB=3}
  }
  MaterialExpressionIf_40 {
   class=MaterialExpressionIf
   inputs [
    A {from=MaterialExpressionReroute_18, output=UVIndex}
    B {from="", output=""}
    "A > B" {from=MaterialExpressionIf_41, output=""}
    "A == B" {from=MaterialExpressionIf_41, output=""}
    "A < B" {from=MaterialExpressionTextureCoordinate_56, output=""}
   ]
   properties {constB=1}
  }
  MaterialExpressionIf_41 {
   class=MaterialExpressionIf
   inputs [
    A {from=MaterialExpressionReroute_18, output=UVIndex}
    B {from="", output=""}
    "A > B" {from=MaterialExpressionIf_42, output=""}
    "A == B" {from=MaterialExpressionIf_42, output=""}
    "A < B" {from=MaterialExpressionTextureCoordinate_57, output=""}
   ]
   properties {constB=2}
  }
  MaterialExpressionIf_42 {
   class=MaterialExpressionIf
   inputs [
    A {from=MaterialExpressionReroute_18, output=UVIndex}
    B {from="", output=""}
    "A > B" {from=MaterialExpressionTextureCoordinate_59, output=""}
    "A == B" {from=MaterialExpressionTextureCoordinate_59, output=""}
    "A < B" {from=MaterialExpressionTextureCoordinate_58, output=""}
   ]
   properties {constB=3}
  }
  MaterialExpressionIf_43 {
   class=MaterialExpressionIf
   inputs [
    A {from=MaterialExpressionReroute_20, output=UVIndex}
    B {from="", output=""}
    "A > B" {from=MaterialExpressionIf_44, output=""}
    "A == B" {from=MaterialExpressionIf_44, output=""}
    "A < B" {from=MaterialExpressionTextureCoordinate_60, output=""}
   ]
   properties {constB=1}
  }
  MaterialExpressionIf_44 {
   class=MaterialExpressionIf
   inputs [
    A {from=MaterialExpressionReroute_20, output=UVIndex}
    B {from="", output=""}
    "A > B" {from=MaterialExpressionIf_45, output=""}
    "A == B" {from=MaterialExpressionIf_45, output=""}
    "A < B" {from=MaterialExpressionTextureCoordinate_61, output=""}
   ]
   properties {constB=2}
  }
  MaterialExpressionIf_45 {
   class=MaterialExpressionIf
   inputs [
    A {from=MaterialExpressionReroute_20, output=UVIndex}
    B {from="", output=""}
    "A > B" {from=MaterialExpressionTextureCoordinate_63, output=""}
    "A == B" {from=MaterialExpressionTextureCoordinate_63, output=""}
    "A < B" {from=MaterialExpressionTextureCoordinate_62, output=""}
   ]
   properties {constB=3}
  }
  MaterialExpressionIf_46 {
   class=MaterialExpressionIf
   inputs [
    A {from=MaterialExpressionReroute_24, output=UVIndex}
    B {from="", output=""}
    "A > B" {from=MaterialExpressionIf_47, output=""}
    "A == B" {from=MaterialExpressionIf_47, output=""}
    "A < B" {from=MaterialExpressionTextureCoordinate_64, output=""}
   ]
   properties {constB=1}
  }
  MaterialExpressionIf_47 {
   class=MaterialExpressionIf
   inputs [
    A {from=MaterialExpressionReroute_24, output=UVIndex}
    B {from="", output=""}
    "A > B" {from=MaterialExpressionIf_48, output=""}
    "A == B" {from=MaterialExpressionIf_48, output=""}
    "A < B" {from=MaterialExpressionTextureCoordinate_65, output=""}
   ]
   properties {constB=2}
  }
  MaterialExpressionIf_48 {
   class=MaterialExpressionIf
   inputs [
    A {from=MaterialExpressionReroute_24, output=UVIndex}
    B {from="", output=""}
    "A > B" {from=MaterialExpressionTextureCoordinate_67, output=""}
    "A == B" {from=MaterialExpressionTextureCoordinate_67, output=""}
    "A < B" {from=MaterialExpressionTextureCoordinate_66, output=""}
   ]
   properties {constB=3}
  }
  MaterialExpressionIf_49 {
   class=MaterialExpressionIf
   inputs [
    A {from=MaterialExpressionReroute_26, output=UVIndex}
    B {from="", output=""}
    "A > B" {from=MaterialExpressionIf_50, output=""}
    "A == B" {from=MaterialExpressionIf_50, output=""}
    "A < B" {from=MaterialExpressionTextureCoordinate_68, output=""}
   ]
   properties {constB=1}
  }
  MaterialExpressionIf_50 {
   class=MaterialExpressionIf
   inputs [
    A {from=MaterialExpressionReroute_26, output=UVIndex}
    B {from="", output=""}
    "A > B" {from=MaterialExpressionIf_51, output=""}
    "A == B" {from=MaterialExpressionIf_51, output=""}
    "A < B" {from=MaterialExpressionTextureCoordinate_69, output=""}
   ]
   properties {constB=2}
  }
  MaterialExpressionIf_51 {
   class=MaterialExpressionIf
   inputs [
    A {from=MaterialExpressionReroute_26, output=UVIndex}
    B {from="", output=""}
    "A > B" {from=MaterialExpressionTextureCoordinate_71, output=""}
    "A == B" {from=MaterialExpressionTextureCoordinate_71, output=""}
    "A < B" {from=MaterialExpressionTextureCoordinate_70, output=""}
   ]
   properties {constB=3}
  }
  MaterialExpressionIf_52 {
   class=MaterialExpressionIf
   inputs [
    A {from=MaterialExpressionReroute_28, output=UVIndex}
    B {from="", output=""}
    "A > B" {from=MaterialExpressionIf_53, output=""}
    "A == B" {from=MaterialExpressionIf_53, output=""}
    "A < B" {from=MaterialExpressionTextureCoordinate_72, output=""}
   ]
   properties {constB=1}
  }
  MaterialExpressionIf_53 {
   class=MaterialExpressionIf
   inputs [
    A {from=MaterialExpressionReroute_28, output=UVIndex}
    B {from="", output=""}
    "A > B" {from=MaterialExpressionIf_54, output=""}
    "A == B" {from=MaterialExpressionIf_54, output=""}
    "A < B" {from=MaterialExpressionTextureCoordinate_73, output=""}
   ]
   properties {constB=2}
  }
  MaterialExpressionIf_54 {
   class=MaterialExpressionIf
   inputs [
    A {from=MaterialExpressionReroute_28, output=UVIndex}
    B {from="", output=""}
    "A > B" {from=MaterialExpressionTextureCoordinate_75, output=""}
    "A == B" {from=MaterialExpressionTextureCoordinate_75, output=""}
    "A < B" {from=MaterialExpressionTextureCoordinate_74, output=""}
   ]
   properties {constB=3}
  }
  MaterialExpressionIf_55 {
   class=MaterialExpressionIf
   inputs [
    A {from=MaterialExpressionReroute_26, output=UVIndex}
    B {from="", output=""}
    "A > B" {from=MaterialExpressionIf_56, output=""}
    "A == B" {from=MaterialExpressionIf_56, output=""}
    "A < B" {from=MaterialExpressionTextureCoordinate_76, output=""}
   ]
   properties {constB=1}
  }
  MaterialExpressionIf_56 {
   class=MaterialExpressionIf
   inputs [
    A {from=MaterialExpressionReroute_26, output=UVIndex}
    B {from="", output=""}
    "A > B" {from=MaterialExpressionIf_57, output=""}
    "A == B" {from=MaterialExpressionIf_57, output=""}
    "A < B" {from=MaterialExpressionTextureCoordinate_77, output=""}
   ]
   properties {constB=2}
  }
  MaterialExpressionIf_57 {
   class=MaterialExpressionIf
   inputs [
    A {from=MaterialExpressionReroute_26, output=UVIndex}
    B {from="", output=""}
    "A > B" {from=MaterialExpressionTextureCoordinate_79, output=""}
    "A == B" {from=MaterialExpressionTextureCoordinate_79, output=""}
    "A < B" {from=MaterialExpressionTextureCoordinate_78, output=""}
   ]
   properties {constB=3}
  }
  MaterialExpressionLinearInterpolate_0 {
   class=MaterialExpressionLinearInterpolate
   inputs [
    A {from=MaterialExpressionLinearInterpolate_8}
    B {from=MaterialExpressionTextureSampleParameter2D_0, output=RGB}
    Alpha {from=MaterialExpressionScalarParameter_3}
   ]
  }
  MaterialExpressionLinearInterpolate_1 {
   class=MaterialExpressionLinearInterpolate
   inputs [
    A {from=MaterialExpressionScalarParameter_0}
    B {from=MaterialExpressionMaterialFunctionCall_4, output=Result}
    Alpha {from=MaterialExpressionScalarParameter_4}
   ]
  }
  MaterialExpressionLinearInterpolate_2 {
   class=MaterialExpressionLinearInterpolate
   inputs [
    A {from=MaterialExpressionScalarParameter_1}
    B {from=MaterialExpressionMaterialFunctionCall_8, output=Result}
    Alpha {from=MaterialExpressionScalarParameter_5}
   ]
  }
  MaterialExpressionLinearInterpolate_3 {
   class=MaterialExpressionLinearInterpolate
   inputs [
    A {from=MaterialExpressionScalarParameter_2}
    B {from=MaterialExpressionMaterialFunctionCall_7, output=Result}
    Alpha {from=MaterialExpressionScalarParameter_6}
   ]
  }
  MaterialExpressionLinearInterpolate_4 {
   class=MaterialExpressionLinearInterpolate
   inputs [
    A {from=MaterialExpressionVectorParameter_2, output=RGB}
    B {from=MaterialExpressionTextureSampleParameter2D_4, output=RGB}
    Alpha {from=MaterialExpressionScalarParameter_7}
   ]
  }
  MaterialExpressionLinearInterpolate_5 {
   class=MaterialExpressionLinearInterpolate
   inputs [
    A {from=MaterialExpressionVectorParameter_1, output=RGB}
    B {from=MaterialExpressionTextureSampleParameter2D_5, output=RGB}
    Alpha {from=MaterialExpressionScalarParameter_8}
   ]
  }
  MaterialExpressionLinearInterpolate_6 {
   class=MaterialExpressionLinearInterpolate
   inputs [
    A {from=MaterialExpressionReroute_1}
    B {from=MaterialExpressionMaterialFunctionCall_0, output=Result}
    Alpha {from=MaterialExpressionScalarParameter_9}
   ]
  }
  MaterialExpressionLinearInterpolate_7 {
   class=MaterialExpressionLinearInterpolate
   inputs [
    A {from=MaterialExpressionScalarParameter_10}
    B {from=MaterialExpressionMaterialFunctionCall_2, output=Result}
    Alpha {from=MaterialExpressionScalarParameter_11}
   ]
  }
  MaterialExpressionLinearInterpolate_8 {
   class=MaterialExpressionLinearInterpolate
   inputs [
    A {from=MaterialExpressionVectorParameter_0, output=RGB}
    B {from=MaterialExpressionMaterialFunctionCall_13, output=Result}
    Alpha {from=MaterialExpressionScalarParameter_12}
   ]
  }
  MaterialExpressionLinearInterpolate_9 {
   class=MaterialExpressionLinearInterpolate
   inputs [
    A {from=MaterialExpressionScalarParameter_13}
    B {from=MaterialExpressionMaterialFunctionCall_14, output=Result}
    Alpha {from=MaterialExpressionScalarParameter_14}
   ]
  }
  MaterialExpressionMaterialFunctionCall_0 {
   class=MaterialExpressionMaterialFunctionCall
   inputs [In {from=MaterialExpressionReroute_1}]
   properties {materialFunction=&/WorldSpaceNormalsToTangentSpace}
  }
  MaterialExpressionMaterialFunctionCall_1 {
   class=MaterialExpressionMaterialFunctionCall
   inputs [
    X {from=MaterialExpressionVectorParameter_3, output=R}
    Y {from=MaterialExpressionVectorParameter_3, output=R}
    Z {from=MaterialExpressionVectorParameter_3, output=R}
    A {from=MaterialExpressionVectorParameter_3, output=R}
   ]
   properties {materialFunction=&MakeFloat4}
  }
  MaterialExpressionMaterialFunctionCall_10 {
   class=MaterialExpressionMaterialFunctionCall
   inputs [sRGB {from=MaterialExpressionVertexColor_0, output=B}]
   properties {materialFunction=/Engine/Functions/Engine_MaterialFunctions01/ImageAdjustment/sRGBToLinear}
  }
  MaterialExpressionMaterialFunctionCall_11 {
   class=MaterialExpressionMaterialFunctionCall
   inputs [sRGB {from=MaterialExpressionVertexColor_0, output=G}]
   properties {materialFunction=&sRGBToLinear}
  }
  MaterialExpressionMaterialFunctionCall_12 {
   class=MaterialExpressionMaterialFunctionCall
   inputs [sRGB {from=MaterialExpressionVertexColor_0, output=R}]
   properties {materialFunction=&sRGBToLinear}
  }
  MaterialExpressionMaterialFunctionCall_13 {
   class=MaterialExpressionMaterialFunctionCall
   inputs [
    X {from=MaterialExpressionMaterialFunctionCall_12, output=Linear}
    Y {from=MaterialExpressionMaterialFunctionCall_11, output=Linear}
    Z {from=MaterialExpressionMaterialFunctionCall_10, output=Linear}
   ]
   properties {materialFunction=&MakeFloat3}
  }
  MaterialExpressionMaterialFunctionCall_14 {
   class=MaterialExpressionMaterialFunctionCall
   inputs [
    Value {from=MaterialExpressionTextureSampleParameter2D_7, output=RGBA}
    Channel {from=MaterialExpressionMaterialFunctionCall_3, output=Result}
   ]
   properties {materialFunction=&/ChannelSelector}
  }
  MaterialExpressionMaterialFunctionCall_17 {
   class=MaterialExpressionMaterialFunctionCall
   inputs [
    Scale {from=MaterialExpressionMaterialFunctionCall_18, output=Result}
    Rotation {from=MaterialExpressionScalarParameter_23}
    TextureCoordinate {from=MaterialExpressionReroute_25, output=UV}
    Translation {from=MaterialExpressionMaterialFunctionCall_19, output=Result}
   ]
   properties {materialFunction=&/UsdTransform2d}
  }
  MaterialExpressionMaterialFunctionCall_18 {
   class=MaterialExpressionMaterialFunctionCall
   inputs [
    X {from=MaterialExpressionVectorParameter_8, output=R}
    Y {from=MaterialExpressionVectorParameter_8, output=R}
   ]
   properties {materialFunction=&MakeFloat2}
  }
  MaterialExpressionMaterialFunctionCall_19 {
   class=MaterialExpressionMaterialFunctionCall
   inputs [
    X {from=MaterialExpressionVectorParameter_8, output=B}
    Y {from=MaterialExpressionVectorParameter_8, output=B}
   ]
   properties {materialFunction=&MakeFloat2}
  }
  MaterialExpressionMaterialFunctionCall_2 {
   class=MaterialExpressionMaterialFunctionCall
   inputs [
    Value {from=MaterialExpressionTextureSampleParameter2D_6, output=RGBA}
    Channel {from=MaterialExpressionMaterialFunctionCall_1, output=Result}
   ]
   properties {materialFunction=&ChannelSelector}
  }
  MaterialExpressionMaterialFunctionCall_20 {
   class=MaterialExpressionMaterialFunctionCall
   inputs [
    X {from=MaterialExpressionVectorParameter_9, output=R}
    Y {from=MaterialExpressionVectorParameter_9, output=R}
   ]
   properties {materialFunction=&MakeFloat2}
  }
  MaterialExpressionMaterialFunctionCall_21 {
   class=MaterialExpressionMaterialFunctionCall
   inputs [
    X {from=MaterialExpressionVectorParameter_9, output=B}
    Y {from=MaterialExpressionVectorParameter_9, output=B}
   ]
   properties {materialFunction=&MakeFloat2}
  }
  MaterialExpressionMaterialFunctionCall_22 {
   class=MaterialExpressionMaterialFunctionCall
   inputs [
    Scale {from=MaterialExpressionMaterialFunctionCall_20, output=Result}
    Rotation {from=MaterialExpressionScalarParameter_24}
    TextureCoordinate {from=MaterialExpressionReroute_17, output=UV}
    Translation {from=MaterialExpressionMaterialFunctionCall_21, output=Result}
   ]
   properties {materialFunction=&UsdTransform2d}
  }
  MaterialExpressionMaterialFunctionCall_23 {
   class=MaterialExpressionMaterialFunctionCall
   inputs [
    X {from=MaterialExpressionVectorParameter_10, output=R}
    Y {from=MaterialExpressionVectorParameter_10, output=R}
   ]
   properties {materialFunction=&MakeFloat2}
  }
  MaterialExpressionMaterialFunctionCall_24 {
   class=MaterialExpressionMaterialFunctionCall
   inputs [
    X {from=MaterialExpressionVectorParameter_10, output=B}
    Y {from=MaterialExpressionVectorParameter_10, output=B}
   ]
   properties {materialFunction=&MakeFloat2}
  }
  MaterialExpressionMaterialFunctionCall_25 {
   class=MaterialExpressionMaterialFunctionCall
   inputs [
    Scale {from=MaterialExpressionMaterialFunctionCall_23, output=Result}
    Rotation {from=MaterialExpressionScalarParameter_25}
    TextureCoordinate {from=MaterialExpressionReroute_19, output=UV}
    Translation {from=MaterialExpressionMaterialFunctionCall_24, output=Result}
   ]
   properties {materialFunction=&UsdTransform2d}
  }
  MaterialExpressionMaterialFunctionCall_26 {
   class=MaterialExpressionMaterialFunctionCall
   inputs [
    X {from=MaterialExpressionVectorParameter_11, output=R}
    Y {from=MaterialExpressionVectorParameter_11, output=R}
   ]
   properties {materialFunction=&MakeFloat2}
  }
  MaterialExpressionMaterialFunctionCall_27 {
   class=MaterialExpressionMaterialFunctionCall
   inputs [
    X {from=MaterialExpressionVectorParameter_11, output=B}
    Y {from=MaterialExpressionVectorParameter_11, output=B}
   ]
   properties {materialFunction=&MakeFloat2}
  }
  MaterialExpressionMaterialFunctionCall_28 {
   class=MaterialExpressionMaterialFunctionCall
   inputs [
    Scale {from=MaterialExpressionMaterialFunctionCall_26, output=Result}
    Rotation {from=MaterialExpressionScalarParameter_26}
    TextureCoordinate {from=MaterialExpressionReroute_21, output=UV}
    Translation {from=MaterialExpressionMaterialFunctionCall_27, output=Result}
   ]
   properties {materialFunction=&UsdTransform2d}
  }
  MaterialExpressionMaterialFunctionCall_29 {
   class=MaterialExpressionMaterialFunctionCall
   inputs [
    X {from=MaterialExpressionVectorParameter_12, output=R}
    Y {from=MaterialExpressionVectorParameter_12, output=R}
   ]
   properties {materialFunction=&MakeFloat2}
  }
  MaterialExpressionMaterialFunctionCall_3 {
   class=MaterialExpressionMaterialFunctionCall
   inputs [
    X {from=MaterialExpressionVectorParameter_4, output=R}
    Y {from=MaterialExpressionVectorParameter_4, output=R}
    Z {from=MaterialExpressionVectorParameter_4, output=R}
    A {from=MaterialExpressionVectorParameter_4, output=R}
   ]
   properties {materialFunction=&MakeFloat4}
  }
  MaterialExpressionMaterialFunctionCall_30 {
   class=MaterialExpressionMaterialFunctionCall
   inputs [
    X {from=MaterialExpressionVectorParameter_12, output=B}
    Y {from=MaterialExpressionVectorParameter_12, output=B}
   ]
   properties {materialFunction=&MakeFloat2}
  }
  MaterialExpressionMaterialFunctionCall_31 {
   class=MaterialExpressionMaterialFunctionCall
   inputs [
    Scale {from=MaterialExpressionMaterialFunctionCall_29, output=Result}
    Rotation {from=MaterialExpressionScalarParameter_27}
    TextureCoordinate {from=MaterialExpressionReroute_25, output=UV}
    Translation {from=MaterialExpressionMaterialFunctionCall_30, output=Result}
   ]
   properties {materialFunction=&UsdTransform2d}
  }
  MaterialExpressionMaterialFunctionCall_32 {
   class=MaterialExpressionMaterialFunctionCall
   inputs [
    X {from=MaterialExpressionVectorParameter_13, output=R}
    Y {from=MaterialExpressionVectorParameter_13, output=R}
   ]
   properties {materialFunction=&MakeFloat2}
  }
  MaterialExpressionMaterialFunctionCall_33 {
   class=MaterialExpressionMaterialFunctionCall
   inputs [
    X {from=MaterialExpressionVectorParameter_13, output=B}
    Y {from=MaterialExpressionVectorParameter_13, output=B}
   ]
   properties {materialFunction=&MakeFloat2}
  }
  MaterialExpressionMaterialFunctionCall_34 {
   class=MaterialExpressionMaterialFunctionCall
   inputs [
    Scale {from=MaterialExpressionMaterialFunctionCall_32, output=Result}
    Rotation {from=MaterialExpressionScalarParameter_28}
    TextureCoordinate {from=MaterialExpressionReroute_27, output=UV}
    Translation {from=MaterialExpressionMaterialFunctionCall_33, output=Result}
   ]
   properties {materialFunction=&UsdTransform2d}
  }
  MaterialExpressionMaterialFunctionCall_35 {
   class=MaterialExpressionMaterialFunctionCall
   inputs [
    X {from=MaterialExpressionVectorParameter_14, output=R}
    Y {from=MaterialExpressionVectorParameter_14, output=R}
   ]
   properties {materialFunction=&MakeFloat2}
  }
  MaterialExpressionMaterialFunctionCall_36 {
   class=MaterialExpressionMaterialFunctionCall
   inputs [
    X {from=MaterialExpressionVectorParameter_14, output=B}
    Y {from=MaterialExpressionVectorParameter_14, output=B}
   ]
   properties {materialFunction=&MakeFloat2}
  }
  MaterialExpressionMaterialFunctionCall_37 {
   class=MaterialExpressionMaterialFunctionCall
   inputs [
    Scale {from=MaterialExpressionMaterialFunctionCall_35, output=Result}
    Rotation {from=MaterialExpressionScalarParameter_29}
    TextureCoordinate {from=MaterialExpressionReroute_29, output=UV}
    Translation {from=MaterialExpressionMaterialFunctionCall_36, output=Result}
   ]
   properties {materialFunction=&UsdTransform2d}
  }
  MaterialExpressionMaterialFunctionCall_38 {
   class=MaterialExpressionMaterialFunctionCall
   inputs [
    X {from=MaterialExpressionVectorParameter_15, output=R}
    Y {from=MaterialExpressionVectorParameter_15, output=R}
   ]
   properties {materialFunction=&MakeFloat2}
  }
  MaterialExpressionMaterialFunctionCall_39 {
   class=MaterialExpressionMaterialFunctionCall
   inputs [
    X {from=MaterialExpressionVectorParameter_15, output=B}
    Y {from=MaterialExpressionVectorParameter_15, output=B}
   ]
   properties {materialFunction=&MakeFloat2}
  }
  MaterialExpressionMaterialFunctionCall_4 {
   class=MaterialExpressionMaterialFunctionCall
   inputs [
    Value {from=MaterialExpressionTextureSampleParameter2D_1, output=RGBA}
    Channel {from=MaterialExpressionMaterialFunctionCall_5, output=Result}
   ]
   properties {materialFunction=&ChannelSelector}
  }
  MaterialExpressionMaterialFunctionCall_40 {
   class=MaterialExpressionMaterialFunctionCall
   inputs [
    Scale {from=MaterialExpressionMaterialFunctionCall_38, output=Result}
    Rotation {from=MaterialExpressionScalarParameter_30}
    TextureCoordinate {from=MaterialExpressionReroute_27, output=UV}
    Translation {from=MaterialExpressionMaterialFunctionCall_39, output=Result}
   ]
   properties {materialFunction=&UsdTransform2d}
  }
  MaterialExpressionMaterialFunctionCall_5 {
   class=MaterialExpressionMaterialFunctionCall
   inputs [
    X {from=MaterialExpressionVectorParameter_5, output=R}
    Y {from=MaterialExpressionVectorParameter_5, output=R}
    Z {from=MaterialExpressionVectorParameter_5, output=R}
    A {from=MaterialExpressionVectorParameter_5, output=R}
   ]
   properties {materialFunction=&MakeFloat4}
  }
  MaterialExpressionMaterialFunctionCall_6 {
   class=MaterialExpressionMaterialFunctionCall
   inputs [
    X {from=MaterialExpressionVectorParameter_6, output=R}
    Y {from=MaterialExpressionVectorParameter_6, output=R}
    Z {from=MaterialExpressionVectorParameter_6, output=R}
    A {from=MaterialExpressionVectorParameter_6, output=R}
   ]
   properties {materialFunction=&MakeFloat4}
  }
  MaterialExpressionMaterialFunctionCall_7 {
   class=MaterialExpressionMaterialFunctionCall
   inputs [
    Value {from=MaterialExpressionTextureSampleParameter2D_3, output=RGBA}
    Channel {from=MaterialExpressionMaterialFunctionCall_6, output=Result}
   ]
   properties {materialFunction=&ChannelSelector}
  }
  MaterialExpressionMaterialFunctionCall_8 {
   class=MaterialExpressionMaterialFunctionCall
   inputs [
    Value {from=MaterialExpressionTextureSampleParameter2D_2, output=RGBA}
    Channel {from=MaterialExpressionMaterialFunctionCall_9, output=Result}
   ]
   properties {materialFunction=&ChannelSelector}
  }
  MaterialExpressionMaterialFunctionCall_9 {
   class=MaterialExpressionMaterialFunctionCall
   inputs [
    X {from=MaterialExpressionVectorParameter_7, output=R}
    Y {from=MaterialExpressionVectorParameter_7, output=R}
    Z {from=MaterialExpressionVectorParameter_7, output=R}
    A {from=MaterialExpressionVectorParameter_7, output=R}
   ]
   properties {materialFunction=&MakeFloat4}
  }
  MaterialExpressionPinBase_14 {
   class=MaterialExpressionPinBase
   inputs []
   properties {reroutePins [UVIndex {expression=&self:MaterialGraph_0.MaterialGraphNode_Composite_12.UVSwitch.MaterialExpressionReroute_24}]}
  }
  MaterialExpressionPinBase_15 {
   class=MaterialExpressionPinBase
   inputs [UV {from=MaterialExpressionStaticSwitchParameter_1, output=""}]
   properties {reroutePins [UV {expression=&self:MaterialGraph_0.MaterialGraphNode_Composite_12.UVSwitch.MaterialExpressionReroute_25}]}
  }
  MaterialExpressionPinBase_24 {
   class=MaterialExpressionPinBase
   inputs []
   properties {reroutePins [UVIndex {expression=&self:MaterialGraph_0.MaterialGraphNode_Composite_13.UVSwitch.MaterialExpressionReroute_16}]}
  }
  MaterialExpressionPinBase_25 {
   class=MaterialExpressionPinBase
   inputs [UV {from=MaterialExpressionStaticSwitchParameter_17, output=""}]
   properties {reroutePins [UV {expression=&self:MaterialGraph_0.MaterialGraphNode_Composite_13.UVSwitch.MaterialExpressionReroute_17}]}
  }
  MaterialExpressionPinBase_26 {
   class=MaterialExpressionPinBase
   inputs []
   properties {reroutePins [UVIndex {expression=&self:MaterialGraph_0.MaterialGraphNode_Composite_14.UVSwitch.MaterialExpressionReroute_18}]}
  }
  MaterialExpressionPinBase_27 {
   class=MaterialExpressionPinBase
   inputs [UV {from=MaterialExpressionStaticSwitchParameter_20, output=""}]
   properties {reroutePins [UV {expression=&self:MaterialGraph_0.MaterialGraphNode_Composite_14.UVSwitch.MaterialExpressionReroute_19}]}
  }
  MaterialExpressionPinBase_28 {
   class=MaterialExpressionPinBase
   inputs []
   properties {reroutePins [UVIndex {expression=&self:MaterialGraph_0.MaterialGraphNode_Composite_15.UVSwitch.MaterialExpressionReroute_20}]}
  }
  MaterialExpressionPinBase_29 {
   class=MaterialExpressionPinBase
   inputs [UV {from=MaterialExpressionStaticSwitchParameter_23, output=""}]
   properties {reroutePins [UV {expression=&self:MaterialGraph_0.MaterialGraphNode_Composite_15.UVSwitch.MaterialExpressionReroute_21}]}
  }
  MaterialExpressionPinBase_30 {
   class=MaterialExpressionPinBase
   inputs []
   properties {reroutePins [UVIndex {expression=&self:MaterialGraph_0.MaterialGraphNode_Composite_16.UVSwitch.MaterialExpressionReroute_24}]}
  }
  MaterialExpressionPinBase_31 {
   class=MaterialExpressionPinBase
   inputs [UV {from=MaterialExpressionStaticSwitchParameter_26, output=""}]
   properties {reroutePins [UV {expression=&self:MaterialGraph_0.MaterialGraphNode_Composite_16.UVSwitch.MaterialExpressionReroute_25}]}
  }
  MaterialExpressionPinBase_32 {
   class=MaterialExpressionPinBase
   inputs []
   properties {reroutePins [UVIndex {expression=&self:MaterialGraph_0.MaterialGraphNode_Composite_17.UVSwitch.MaterialExpressionReroute_26}]}
  }
  MaterialExpressionPinBase_33 {
   class=MaterialExpressionPinBase
   inputs [UV {from=MaterialExpressionStaticSwitchParameter_29, output=""}]
   properties {reroutePins [UV {expression=&self:MaterialGraph_0.MaterialGraphNode_Composite_17.UVSwitch.MaterialExpressionReroute_27}]}
  }
  MaterialExpressionPinBase_34 {
   class=MaterialExpressionPinBase
   inputs []
   properties {reroutePins [UVIndex {expression=&self:MaterialGraph_0.MaterialGraphNode_Composite_18.UVSwitch.MaterialExpressionReroute_28}]}
  }
  MaterialExpressionPinBase_35 {
   class=MaterialExpressionPinBase
   inputs [UV {from=MaterialExpressionStaticSwitchParameter_32, output=""}]
   properties {reroutePins [UV {expression=&self:MaterialGraph_0.MaterialGraphNode_Composite_18.UVSwitch.MaterialExpressionReroute_29}]}
  }
  MaterialExpressionPinBase_36 {
   class=MaterialExpressionPinBase
   inputs []
   properties {reroutePins [UVIndex {expression=&self:MaterialExpressionReroute_26}]}
  }
  MaterialExpressionPinBase_37 {
   class=MaterialExpressionPinBase
   inputs [UV {from=MaterialExpressionStaticSwitchParameter_35, output=""}]
   properties {reroutePins [UV {expression=&self:MaterialExpressionReroute_27}]}
  }
  MaterialExpressionReroute_1 {class=MaterialExpressionReroute, inputs [None {from=MaterialExpressionLinearInterpolate_4}]}
  MaterialExpressionReroute_26 {class=MaterialExpressionReroute, inputs [None {from=MaterialExpressionScalarParameter_22}]}
  MaterialExpressionReroute_27 {class=MaterialExpressionReroute, inputs [None {from=MaterialExpressionStaticSwitchParameter_35}]}
  MaterialExpressionScalarParameter_0 {class=MaterialExpressionScalarParameter, properties {parameterName=Metallic, sliderMax=1}}
  MaterialExpressionScalarParameter_1 {class=MaterialExpressionScalarParameter, properties {defaultValue=1, parameterName=Roughness, sliderMax=1}}
  MaterialExpressionScalarParameter_10 {class=MaterialExpressionScalarParameter, properties {defaultValue=1.5, parameterName=Refraction, sliderMax=1}}
  MaterialExpressionScalarParameter_11 {class=MaterialExpressionScalarParameter, properties {parameterName=UseRefractionTexture}}
  MaterialExpressionScalarParameter_12 {class=MaterialExpressionScalarParameter, properties {parameterName=UseVertexColorForBaseColor}}
  MaterialExpressionScalarParameter_13 {class=MaterialExpressionScalarParameter, properties {defaultValue=1, parameterName=AmbientOcclusion, sliderMax=1}}
  MaterialExpressionScalarParameter_14 {class=MaterialExpressionScalarParameter, properties {parameterName=UseAmbientOcclusionTexture}}
  MaterialExpressionScalarParameter_15 {class=MaterialExpressionScalarParameter, properties {parameterName=BaseColorUVIndex}}
  MaterialExpressionScalarParameter_16 {class=MaterialExpressionScalarParameter, properties {parameterName=MetallicUVIndex}}
  MaterialExpressionScalarParameter_17 {class=MaterialExpressionScalarParameter, properties {parameterName=RoughnessUVIndex}}
  MaterialExpressionScalarParameter_18 {class=MaterialExpressionScalarParameter, properties {parameterName=EmissiveUVIndex}}
  MaterialExpressionScalarParameter_19 {class=MaterialExpressionScalarParameter, properties {parameterName=OpacityUVIndex}}
  MaterialExpressionScalarParameter_2 {class=MaterialExpressionScalarParameter, properties {defaultValue=1, parameterName=Opacity, sliderMax=1}}
  MaterialExpressionScalarParameter_20 {class=MaterialExpressionScalarParameter, properties {parameterName=NormalUVIndex}}
  MaterialExpressionScalarParameter_21 {class=MaterialExpressionScalarParameter, properties {parameterName=RefractionUVIndex}}
  MaterialExpressionScalarParameter_22 {class=MaterialExpressionScalarParameter, properties {parameterName=AmbientOcclusionUVIndex}}
  MaterialExpressionScalarParameter_23 {class=MaterialExpressionScalarParameter, properties {parameterName=BaseColorRotation}}
  MaterialExpressionScalarParameter_24 {class=MaterialExpressionScalarParameter, properties {parameterName=MetallicRotation}}
  MaterialExpressionScalarParameter_25 {class=MaterialExpressionScalarParameter, properties {parameterName=RoughnessRotation}}
  MaterialExpressionScalarParameter_26 {class=MaterialExpressionScalarParameter, properties {parameterName=EmissiveRotation}}
  MaterialExpressionScalarParameter_27 {class=MaterialExpressionScalarParameter, properties {parameterName=OpacityRotation}}
  MaterialExpressionScalarParameter_28 {class=MaterialExpressionScalarParameter, properties {parameterName=NormalRotation}}
  MaterialExpressionScalarParameter_29 {class=MaterialExpressionScalarParameter, properties {parameterName=RefractionRotation}}
  MaterialExpressionScalarParameter_3 {class=MaterialExpressionScalarParameter, properties {parameterName=UseBaseColorTexture}}
  MaterialExpressionScalarParameter_30 {class=MaterialExpressionScalarParameter, properties {parameterName=AmbientOcclusionRotation}}
  MaterialExpressionScalarParameter_4 {class=MaterialExpressionScalarParameter, properties {parameterName=UseMetallicTexture}}
  MaterialExpressionScalarParameter_5 {class=MaterialExpressionScalarParameter, properties {parameterName=UseRoughnessTexture}}
  MaterialExpressionScalarParameter_6 {class=MaterialExpressionScalarParameter, properties {parameterName=UseOpacityTexture}}
  MaterialExpressionScalarParameter_7 {class=MaterialExpressionScalarParameter, properties {parameterName=UseNormalTexture}}
  MaterialExpressionScalarParameter_8 {class=MaterialExpressionScalarParameter, properties {parameterName=UseEmissiveColorTexture}}
  MaterialExpressionScalarParameter_9 {
   class=MaterialExpressionScalarParameter
   properties {desc=Not a static switch so that we can do this at runtime, parameterName=UseWorldSpaceNormals}
  }
  MaterialExpressionStaticSwitchParameter_1 {
   class=MaterialExpressionStaticSwitchParameter
   inputs [
    True {from=MaterialExpressionSwitch_0, output=""}
    False {from=MaterialExpressionIf_22, output=""}
   ]
   properties {
    desc=Fallback to use 'if' nodes instead of switch expression. Safer way to have analytical derivatives with nanite
    group=Nanite
    parameterName=bUseSwitch
   }
  }
  MaterialExpressionStaticSwitchParameter_17 {
   class=MaterialExpressionStaticSwitchParameter
   inputs [
    True {from=MaterialExpressionSwitch_5, output=""}
    False {from=MaterialExpressionIf_37, output=""}
   ]
   properties {
    desc=Fallback to use 'if' nodes instead of switch expression. Safer way to have analytical derivatives with nanite
    group=Nanite
    parameterName=bUseSwitch
   }
  }
  MaterialExpressionStaticSwitchParameter_20 {
   class=MaterialExpressionStaticSwitchParameter
   inputs [
    True {from=MaterialExpressionSwitch_6, output=""}
    False {from=MaterialExpressionIf_40, output=""}
   ]
   properties {
    desc=Fallback to use 'if' nodes instead of switch expression. Safer way to have analytical derivatives with nanite
    group=Nanite
    parameterName=bUseSwitch
   }
  }
  MaterialExpressionStaticSwitchParameter_23 {
   class=MaterialExpressionStaticSwitchParameter
   inputs [
    True {from=MaterialExpressionSwitch_7, output=""}
    False {from=MaterialExpressionIf_43, output=""}
   ]
   properties {
    desc=Fallback to use 'if' nodes instead of switch expression. Safer way to have analytical derivatives with nanite
    group=Nanite
    parameterName=bUseSwitch
   }
  }
  MaterialExpressionStaticSwitchParameter_26 {
   class=MaterialExpressionStaticSwitchParameter
   inputs [
    True {from=MaterialExpressionSwitch_8, output=""}
    False {from=MaterialExpressionIf_46, output=""}
   ]
   properties {
    desc=Fallback to use 'if' nodes instead of switch expression. Safer way to have analytical derivatives with nanite
    group=Nanite
    parameterName=bUseSwitch
   }
  }
  MaterialExpressionStaticSwitchParameter_29 {
   class=MaterialExpressionStaticSwitchParameter
   inputs [
    True {from=MaterialExpressionSwitch_9, output=""}
    False {from=MaterialExpressionIf_49, output=""}
   ]
   properties {
    desc=Fallback to use 'if' nodes instead of switch expression. Safer way to have analytical derivatives with nanite
    group=Nanite
    parameterName=bUseSwitch
   }
  }
  MaterialExpressionStaticSwitchParameter_32 {
   class=MaterialExpressionStaticSwitchParameter
   inputs [
    True {from=MaterialExpressionSwitch_10, output=""}
    False {from=MaterialExpressionIf_52, output=""}
   ]
   properties {
    desc=Fallback to use 'if' nodes instead of switch expression. Safer way to have analytical derivatives with nanite
    group=Nanite
    parameterName=bUseSwitch
   }
  }
  MaterialExpressionStaticSwitchParameter_35 {
   class=MaterialExpressionStaticSwitchParameter
   inputs [
    True {from=MaterialExpressionSwitch_11, output=""}
    False {from=MaterialExpressionIf_55, output=""}
   ]
   properties {
    desc=Fallback to use 'if' nodes instead of switch expression. Safer way to have analytical derivatives with nanite
    group=Nanite
    parameterName=bUseSwitch
   }
  }
  MaterialExpressionSwitch_0 {
   class=MaterialExpressionSwitch
   inputs [
    SwitchValue {from=MaterialExpressionReroute_24, output=UVIndex}
    Default {from=MaterialExpressionTextureCoordinate_32, output=""}
    UV0 {from=MaterialExpressionTextureCoordinate_32, output=""}
    UV1 {from=MaterialExpressionTextureCoordinate_33, output=""}
    UV2 {from=MaterialExpressionTextureCoordinate_34, output=""}
    UV3 {from=MaterialExpressionTextureCoordinate_35, output=""}
   ]
   properties {inputs [
     {
      inputName=UV0
      input {
       expression=&self:MaterialExpressionTextureCoordinate_32
       inputName=None
       mask=0
       maskA=0
       maskB=0
       maskG=0
       maskR=0
       outputIndex=0
      }
     }
     {
      inputName=UV1
      input {
       expression=&self:MaterialExpressionTextureCoordinate_33
       inputName=None
       mask=0
       maskA=0
       maskB=0
       maskG=0
       maskR=0
       outputIndex=0
      }
     }
     {
      inputName=UV2
      input {
       expression=&self:MaterialExpressionTextureCoordinate_34
       inputName=None
       mask=0
       maskA=0
       maskB=0
       maskG=0
       maskR=0
       outputIndex=0
      }
     }
     {
      inputName=UV3
      input {
       expression=&self:MaterialExpressionTextureCoordinate_35
       inputName=None
       mask=0
       maskA=0
       maskB=0
       maskG=0
       maskR=0
       outputIndex=0
      }
     }
    ]}
  }
  MaterialExpressionSwitch_10 {
   class=MaterialExpressionSwitch
   inputs [
    SwitchValue {from=MaterialExpressionReroute_28, output=UVIndex}
    Default {from=MaterialExpressionTextureCoordinate_72, output=""}
    UV0 {from=MaterialExpressionTextureCoordinate_72, output=""}
    UV1 {from=MaterialExpressionTextureCoordinate_73, output=""}
    UV2 {from=MaterialExpressionTextureCoordinate_74, output=""}
    UV3 {from=MaterialExpressionTextureCoordinate_75, output=""}
   ]
   properties {inputs [
     {
      inputName=UV0
      input {
       expression=&self:MaterialExpressionTextureCoordinate_72
       inputName=None
       mask=0
       maskA=0
       maskB=0
       maskG=0
       maskR=0
       outputIndex=0
      }
     }
     {
      inputName=UV1
      input {
       expression=&self:MaterialExpressionTextureCoordinate_73
       inputName=None
       mask=0
       maskA=0
       maskB=0
       maskG=0
       maskR=0
       outputIndex=0
      }
     }
     {
      inputName=UV2
      input {
       expression=&self:MaterialExpressionTextureCoordinate_74
       inputName=None
       mask=0
       maskA=0
       maskB=0
       maskG=0
       maskR=0
       outputIndex=0
      }
     }
     {
      inputName=UV3
      input {
       expression=&self:MaterialExpressionTextureCoordinate_75
       inputName=None
       mask=0
       maskA=0
       maskB=0
       maskG=0
       maskR=0
       outputIndex=0
      }
     }
    ]}
  }
  MaterialExpressionSwitch_11 {
   class=MaterialExpressionSwitch
   inputs [
    SwitchValue {from=MaterialExpressionReroute_26, output=UVIndex}
    Default {from=MaterialExpressionTextureCoordinate_76, output=""}
    UV0 {from=MaterialExpressionTextureCoordinate_76, output=""}
    UV1 {from=MaterialExpressionTextureCoordinate_77, output=""}
    UV2 {from=MaterialExpressionTextureCoordinate_78, output=""}
    UV3 {from=MaterialExpressionTextureCoordinate_79, output=""}
   ]
   properties {inputs [
     {
      inputName=UV0
      input {
       expression=&self:MaterialExpressionTextureCoordinate_76
       inputName=None
       mask=0
       maskA=0
       maskB=0
       maskG=0
       maskR=0
       outputIndex=0
      }
     }
     {
      inputName=UV1
      input {
       expression=&self:MaterialExpressionTextureCoordinate_77
       inputName=None
       mask=0
       maskA=0
       maskB=0
       maskG=0
       maskR=0
       outputIndex=0
      }
     }
     {
      inputName=UV2
      input {
       expression=&self:MaterialExpressionTextureCoordinate_78
       inputName=None
       mask=0
       maskA=0
       maskB=0
       maskG=0
       maskR=0
       outputIndex=0
      }
     }
     {
      inputName=UV3
      input {
       expression=&self:MaterialExpressionTextureCoordinate_79
       inputName=None
       mask=0
       maskA=0
       maskB=0
       maskG=0
       maskR=0
       outputIndex=0
      }
     }
    ]}
  }
  MaterialExpressionSwitch_5 {
   class=MaterialExpressionSwitch
   inputs [
    SwitchValue {from=MaterialExpressionReroute_16, output=UVIndex}
    Default {from=MaterialExpressionTextureCoordinate_52, output=""}
    UV0 {from=MaterialExpressionTextureCoordinate_52, output=""}
    UV1 {from=MaterialExpressionTextureCoordinate_53, output=""}
    UV2 {from=MaterialExpressionTextureCoordinate_54, output=""}
    UV3 {from=MaterialExpressionTextureCoordinate_55, output=""}
   ]
   properties {inputs [
     {
      inputName=UV0
      input {
       expression=&self:MaterialExpressionTextureCoordinate_52
       inputName=None
       mask=0
       maskA=0
       maskB=0
       maskG=0
       maskR=0
       outputIndex=0
      }
     }
     {
      inputName=UV1
      input {
       expression=&self:MaterialExpressionTextureCoordinate_53
       inputName=None
       mask=0
       maskA=0
       maskB=0
       maskG=0
       maskR=0
       outputIndex=0
      }
     }
     {
      inputName=UV2
      input {
       expression=&self:MaterialExpressionTextureCoordinate_54
       inputName=None
       mask=0
       maskA=0
       maskB=0
       maskG=0
       maskR=0
       outputIndex=0
      }
     }
     {
      inputName=UV3
      input {
       expression=&self:MaterialExpressionTextureCoordinate_55
       inputName=None
       mask=0
       maskA=0
       maskB=0
       maskG=0
       maskR=0
       outputIndex=0
      }
     }
    ]}
  }
  MaterialExpressionSwitch_6 {
   class=MaterialExpressionSwitch
   inputs [
    SwitchValue {from=MaterialExpressionReroute_18, output=UVIndex}
    Default {from=MaterialExpressionTextureCoordinate_56, output=""}
    UV0 {from=MaterialExpressionTextureCoordinate_56, output=""}
    UV1 {from=MaterialExpressionTextureCoordinate_57, output=""}
    UV2 {from=MaterialExpressionTextureCoordinate_58, output=""}
    UV3 {from=MaterialExpressionTextureCoordinate_59, output=""}
   ]
   properties {inputs [
     {
      inputName=UV0
      input {
       expression=&self:MaterialExpressionTextureCoordinate_56
       inputName=None
       mask=0
       maskA=0
       maskB=0
       maskG=0
       maskR=0
       outputIndex=0
      }
     }
     {
      inputName=UV1
      input {
       expression=&self:MaterialExpressionTextureCoordinate_57
       inputName=None
       mask=0
       maskA=0
       maskB=0
       maskG=0
       maskR=0
       outputIndex=0
      }
     }
     {
      inputName=UV2
      input {
       expression=&self:MaterialExpressionTextureCoordinate_58
       inputName=None
       mask=0
       maskA=0
       maskB=0
       maskG=0
       maskR=0
       outputIndex=0
      }
     }
     {
      inputName=UV3
      input {
       expression=&self:MaterialExpressionTextureCoordinate_59
       inputName=None
       mask=0
       maskA=0
       maskB=0
       maskG=0
       maskR=0
       outputIndex=0
      }
     }
    ]}
  }
  MaterialExpressionSwitch_7 {
   class=MaterialExpressionSwitch
   inputs [
    SwitchValue {from=MaterialExpressionReroute_20, output=UVIndex}
    Default {from=MaterialExpressionTextureCoordinate_60, output=""}
    UV0 {from=MaterialExpressionTextureCoordinate_60, output=""}
    UV1 {from=MaterialExpressionTextureCoordinate_61, output=""}
    UV2 {from=MaterialExpressionTextureCoordinate_62, output=""}
    UV3 {from=MaterialExpressionTextureCoordinate_63, output=""}
   ]
   properties {inputs [
     {
      inputName=UV0
      input {
       expression=&self:MaterialExpressionTextureCoordinate_60
       inputName=None
       mask=0
       maskA=0
       maskB=0
       maskG=0
       maskR=0
       outputIndex=0
      }
     }
     {
      inputName=UV1
      input {
       expression=&self:MaterialExpressionTextureCoordinate_61
       inputName=None
       mask=0
       maskA=0
       maskB=0
       maskG=0
       maskR=0
       outputIndex=0
      }
     }
     {
      inputName=UV2
      input {
       expression=&self:MaterialExpressionTextureCoordinate_62
       inputName=None
       mask=0
       maskA=0
       maskB=0
       maskG=0
       maskR=0
       outputIndex=0
      }
     }
     {
      inputName=UV3
      input {
       expression=&self:MaterialExpressionTextureCoordinate_63
       inputName=None
       mask=0
       maskA=0
       maskB=0
       maskG=0
       maskR=0
       outputIndex=0
      }
     }
    ]}
  }
  MaterialExpressionSwitch_8 {
   class=MaterialExpressionSwitch
   inputs [
    SwitchValue {from=MaterialExpressionReroute_24, output=UVIndex}
    Default {from=MaterialExpressionTextureCoordinate_64, output=""}
    UV0 {from=MaterialExpressionTextureCoordinate_64, output=""}
    UV1 {from=MaterialExpressionTextureCoordinate_65, output=""}
    UV2 {from=MaterialExpressionTextureCoordinate_66, output=""}
    UV3 {from=MaterialExpressionTextureCoordinate_67, output=""}
   ]
   properties {inputs [
     {
      inputName=UV0
      input {
       expression=&self:MaterialExpressionTextureCoordinate_64
       inputName=None
       mask=0
       maskA=0
       maskB=0
       maskG=0
       maskR=0
       outputIndex=0
      }
     }
     {
      inputName=UV1
      input {
       expression=&self:MaterialExpressionTextureCoordinate_65
       inputName=None
       mask=0
       maskA=0
       maskB=0
       maskG=0
       maskR=0
       outputIndex=0
      }
     }
     {
      inputName=UV2
      input {
       expression=&self:MaterialExpressionTextureCoordinate_66
       inputName=None
       mask=0
       maskA=0
       maskB=0
       maskG=0
       maskR=0
       outputIndex=0
      }
     }
     {
      inputName=UV3
      input {
       expression=&self:MaterialExpressionTextureCoordinate_67
       inputName=None
       mask=0
       maskA=0
       maskB=0
       maskG=0
       maskR=0
       outputIndex=0
      }
     }
    ]}
  }
  MaterialExpressionSwitch_9 {
   class=MaterialExpressionSwitch
   inputs [
    SwitchValue {from=MaterialExpressionReroute_26, output=UVIndex}
    Default {from=MaterialExpressionTextureCoordinate_68, output=""}
    UV0 {from=MaterialExpressionTextureCoordinate_68, output=""}
    UV1 {from=MaterialExpressionTextureCoordinate_69, output=""}
    UV2 {from=MaterialExpressionTextureCoordinate_70, output=""}
    UV3 {from=MaterialExpressionTextureCoordinate_71, output=""}
   ]
   properties {inputs [
     {
      inputName=UV0
      input {
       expression=&self:MaterialExpressionTextureCoordinate_68
       inputName=None
       mask=0
       maskA=0
       maskB=0
       maskG=0
       maskR=0
       outputIndex=0
      }
     }
     {
      inputName=UV1
      input {
       expression=&self:MaterialExpressionTextureCoordinate_69
       inputName=None
       mask=0
       maskA=0
       maskB=0
       maskG=0
       maskR=0
       outputIndex=0
      }
     }
     {
      inputName=UV2
      input {
       expression=&self:MaterialExpressionTextureCoordinate_70
       inputName=None
       mask=0
       maskA=0
       maskB=0
       maskG=0
       maskR=0
       outputIndex=0
      }
     }
     {
      inputName=UV3
      input {
       expression=&self:MaterialExpressionTextureCoordinate_71
       inputName=None
       mask=0
       maskA=0
       maskB=0
       maskG=0
       maskR=0
       outputIndex=0
      }
     }
    ]}
  }
  MaterialExpressionTextureCoordinate_32 {class=MaterialExpressionTextureCoordinate}
  MaterialExpressionTextureCoordinate_33 {class=MaterialExpressionTextureCoordinate, properties {coordinateIndex=1}}
  MaterialExpressionTextureCoordinate_34 {class=MaterialExpressionTextureCoordinate, properties {coordinateIndex=2}}
  MaterialExpressionTextureCoordinate_35 {class=MaterialExpressionTextureCoordinate, properties {coordinateIndex=3}}
  MaterialExpressionTextureCoordinate_52 {class=MaterialExpressionTextureCoordinate}
  MaterialExpressionTextureCoordinate_53 {class=MaterialExpressionTextureCoordinate, properties {coordinateIndex=1}}
  MaterialExpressionTextureCoordinate_54 {class=MaterialExpressionTextureCoordinate, properties {coordinateIndex=2}}
  MaterialExpressionTextureCoordinate_55 {class=MaterialExpressionTextureCoordinate, properties {coordinateIndex=3}}
  MaterialExpressionTextureCoordinate_56 {class=MaterialExpressionTextureCoordinate}
  MaterialExpressionTextureCoordinate_57 {class=MaterialExpressionTextureCoordinate, properties {coordinateIndex=1}}
  MaterialExpressionTextureCoordinate_58 {class=MaterialExpressionTextureCoordinate, properties {coordinateIndex=2}}
  MaterialExpressionTextureCoordinate_59 {class=MaterialExpressionTextureCoordinate, properties {coordinateIndex=3}}
  MaterialExpressionTextureCoordinate_60 {class=MaterialExpressionTextureCoordinate}
  MaterialExpressionTextureCoordinate_61 {class=MaterialExpressionTextureCoordinate, properties {coordinateIndex=1}}
  MaterialExpressionTextureCoordinate_62 {class=MaterialExpressionTextureCoordinate, properties {coordinateIndex=2}}
  MaterialExpressionTextureCoordinate_63 {class=MaterialExpressionTextureCoordinate, properties {coordinateIndex=3}}
  MaterialExpressionTextureCoordinate_64 {class=MaterialExpressionTextureCoordinate}
  MaterialExpressionTextureCoordinate_65 {class=MaterialExpressionTextureCoordinate, properties {coordinateIndex=1}}
  MaterialExpressionTextureCoordinate_66 {class=MaterialExpressionTextureCoordinate, properties {coordinateIndex=2}}
  MaterialExpressionTextureCoordinate_67 {class=MaterialExpressionTextureCoordinate, properties {coordinateIndex=3}}
  MaterialExpressionTextureCoordinate_68 {class=MaterialExpressionTextureCoordinate}
  MaterialExpressionTextureCoordinate_69 {class=MaterialExpressionTextureCoordinate, properties {coordinateIndex=1}}
  MaterialExpressionTextureCoordinate_70 {class=MaterialExpressionTextureCoordinate, properties {coordinateIndex=2}}
  MaterialExpressionTextureCoordinate_71 {class=MaterialExpressionTextureCoordinate, properties {coordinateIndex=3}}
  MaterialExpressionTextureCoordinate_72 {class=MaterialExpressionTextureCoordinate}
  MaterialExpressionTextureCoordinate_73 {class=MaterialExpressionTextureCoordinate, properties {coordinateIndex=1}}
  MaterialExpressionTextureCoordinate_74 {class=MaterialExpressionTextureCoordinate, properties {coordinateIndex=2}}
  MaterialExpressionTextureCoordinate_75 {class=MaterialExpressionTextureCoordinate, properties {coordinateIndex=3}}
  MaterialExpressionTextureCoordinate_76 {class=MaterialExpressionTextureCoordinate}
  MaterialExpressionTextureCoordinate_77 {class=MaterialExpressionTextureCoordinate, properties {coordinateIndex=1}}
  MaterialExpressionTextureCoordinate_78 {class=MaterialExpressionTextureCoordinate, properties {coordinateIndex=2}}
  MaterialExpressionTextureCoordinate_79 {class=MaterialExpressionTextureCoordinate, properties {coordinateIndex=3}}
  MaterialExpressionTextureSampleParameter2D_0 {
   class=MaterialExpressionTextureSampleParameter2D
   inputs [
    UVs {from=MaterialExpressionMaterialFunctionCall_17, output=Result}
    Apply View MipBias {from="", output=""}
   ]
   properties {parameterName=BaseColorTexture, texture=/Engine/EditorLandscapeResources/WhiteSquareTexture}
  }
  MaterialExpressionTextureSampleParameter2D_1 {
   class=MaterialExpressionTextureSampleParameter2D
   inputs [
    UVs {from=MaterialExpressionMaterialFunctionCall_22, output=Result}
    Apply View MipBias {from="", output=""}
   ]
   properties {parameterName=MetallicTexture, samplerType=SAMPLERTYPE_LinearColor, texture=/Engine/ArtTools/RenderToTexture/Textures/127grey}
  }
  MaterialExpressionTextureSampleParameter2D_2 {
   class=MaterialExpressionTextureSampleParameter2D
   inputs [
    UVs {from=MaterialExpressionMaterialFunctionCall_25, output=Result}
    Apply View MipBias {from="", output=""}
   ]
   properties {parameterName=RoughnessTexture, samplerType=SAMPLERTYPE_LinearColor, texture=&127grey}
  }
  MaterialExpressionTextureSampleParameter2D_3 {
   class=MaterialExpressionTextureSampleParameter2D
   inputs [
    UVs {from=MaterialExpressionMaterialFunctionCall_31, output=Result}
    Apply View MipBias {from="", output=""}
   ]
   properties {parameterName=OpacityTexture, samplerType=SAMPLERTYPE_LinearColor, texture=&127grey}
  }
  MaterialExpressionTextureSampleParameter2D_4 {
   class=MaterialExpressionTextureSampleParameter2D
   inputs [
    UVs {from=MaterialExpressionMaterialFunctionCall_34, output=Result}
    Apply View MipBias {from="", output=""}
   ]
   properties {parameterName=NormalTexture, samplerType=SAMPLERTYPE_Normal, texture=/Engine/EngineMaterials/BaseFlattenNormalMap}
  }
  MaterialExpressionTextureSampleParameter2D_5 {
   class=MaterialExpressionTextureSampleParameter2D
   inputs [
    UVs {from=MaterialExpressionMaterialFunctionCall_28, output=Result}
    Apply View MipBias {from="", output=""}
   ]
   properties {parameterName=EmissiveColorTexture, texture=&WhiteSquareTexture}
  }
  MaterialExpressionTextureSampleParameter2D_6 {
   class=MaterialExpressionTextureSampleParameter2D
   inputs [
    UVs {from=MaterialExpressionMaterialFunctionCall_37, output=Result}
    Apply View MipBias {from="", output=""}
   ]
   properties {parameterName=RefractionTexture, samplerType=SAMPLERTYPE_LinearColor, texture=&127grey}
  }
  MaterialExpressionTextureSampleParameter2D_7 {
   class=MaterialExpressionTextureSampleParameter2D
   inputs [
    UVs {from=MaterialExpressionMaterialFunctionCall_40, output=Result}
    Apply View MipBias {from="", output=""}
   ]
   properties {parameterName=AmbientOcclusionTexture, samplerType=SAMPLERTYPE_LinearColor, texture=&127grey}
  }
  MaterialExpressionVectorParameter_0 {class=MaterialExpressionVectorParameter, properties {parameterName=BaseColor, defaultValue=<rgba 0 0 0 1>}}
  MaterialExpressionVectorParameter_1 {class=MaterialExpressionVectorParameter, properties {parameterName=EmissiveColor, defaultValue=<rgba 0 0 0 1>}}
  MaterialExpressionVectorParameter_10 {class=MaterialExpressionVectorParameter, properties {parameterName=RoughnessScaleTranslation, defaultValue=<rgba 1 1 0 0>}}
  MaterialExpressionVectorParameter_11 {class=MaterialExpressionVectorParameter, properties {parameterName=EmissiveScaleTranslation, defaultValue=<rgba 1 1 0 0>}}
  MaterialExpressionVectorParameter_12 {class=MaterialExpressionVectorParameter, properties {parameterName=OpacityScaleTranslation, defaultValue=<rgba 1 1 0 0>}}
  MaterialExpressionVectorParameter_13 {class=MaterialExpressionVectorParameter, properties {parameterName=NormalScaleTranslation, defaultValue=<rgba 1 1 0 0>}}
  MaterialExpressionVectorParameter_14 {class=MaterialExpressionVectorParameter, properties {parameterName=RefractionScaleTranslation, defaultValue=<rgba 1 1 0 0>}}
  MaterialExpressionVectorParameter_15 {class=MaterialExpressionVectorParameter, properties {parameterName=AmbientOcclusionScaleTranslation, defaultValue=<rgba 1 1 0 0>}}
  MaterialExpressionVectorParameter_2 {class=MaterialExpressionVectorParameter, properties {parameterName=Normal, defaultValue=<rgba 0 0 1 1>}}
  MaterialExpressionVectorParameter_3 {class=MaterialExpressionVectorParameter, properties {parameterName=RefractionTextureComponent, defaultValue=<rgba 1 0 0 0>}}
  MaterialExpressionVectorParameter_4 {class=MaterialExpressionVectorParameter, properties {parameterName=AmbientOcclusionTextureComponent, defaultValue=<rgba 1 0 0 0>}}
  MaterialExpressionVectorParameter_5 {class=MaterialExpressionVectorParameter, properties {parameterName=MetallicTextureComponent, defaultValue=<rgba 1 0 0 0>}}
  MaterialExpressionVectorParameter_6 {class=MaterialExpressionVectorParameter, properties {parameterName=OpacityTextureComponent, defaultValue=<rgba 1 0 0 0>}}
  MaterialExpressionVectorParameter_7 {class=MaterialExpressionVectorParameter, properties {parameterName=RoughnessTextureComponent, defaultValue=<rgba 1 0 0 0>}}
  MaterialExpressionVectorParameter_8 {class=MaterialExpressionVectorParameter, properties {parameterName=BaseColorScaleTranslation, defaultValue=<rgba 1 1 0 0>}}
  MaterialExpressionVectorParameter_9 {class=MaterialExpressionVectorParameter, properties {parameterName=MetallicScaleTranslation, defaultValue=<rgba 1 1 0 0>}}
  MaterialExpressionVertexColor_0 {class=MaterialExpressionVertexColor, inputs [], properties {}}
 }
 properties {
  bUsedWithClothing=true
  bUsedWithGeometryCache=true
  bUsedWithInstancedStaticMeshes=true
  bUsedWithMeshDeformer=true
  bUsedWithMorphTargets=true
  bUsedWithNanite=true
  bUsedWithSkeletalMesh=true
  bUsedWithVoxels=true
  pixelDepthOffsetMode=PDOM_Legacy
  refractionMethod=RM_IndexOfRefraction
  translucencyLightingMode=TLM_SurfacePerPixelLighting
  usedShadingModels=Default Lit
 }
}
