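# Parent-style USD Preview Surface material using virtual texture samplers, exposing base color, metallic, roughness, opacity, emissive, normal, refraction and ambient occlusion.
# Each channel lerps between a constant and a texture via a scalar Use parameter; base color can also take linearized vertex color.
# Custom HLSL selects one of four UV sets per texture before UsdTransform2d, and ChannelSelector extracts scalar components.
# A runtime scalar, not static switch, enables world-to-tangent normal conversion.
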Material /USDCore/Materials/UsdPreviewSurfaceVT
graph {
 properties {
  bUsedWithGeometryCache=true
  bUsedWithInstancedStaticMeshes=true
  bUsedWithMeshDeformer=true
  bUsedWithMorphTargets=true
  bUsedWithSkeletalMesh=true
  pixelDepthOffsetMode=PDOM_Legacy
  refractionMethod=RM_IndexOfRefraction
  usedShadingModels=Default Lit
 }
 nodes [
  n1 = VertexColor()
  n2 = MaterialFunctionCall!(sRGB=@n1.R) {materialFunction=/Engine/Functions/Engine_MaterialFunctions01/ImageAdjustment/sRGBToLinear}
  n3 = MaterialFunctionCall(sRGB=@n1.G) {materialFunction=&sRGBToLinear}
  n4 = MaterialFunctionCall(sRGB=@n1.B) {materialFunction=&sRGBToLinear}
  n5 = VectorParameter() {parameterName=BaseColorScaleTranslation, defaultValue=<rgba 1 1 0 0>}
  n6 = MaterialFunctionCall(X=@n5.R, Y=@n5.R) {materialFunction=&MakeFloat2}
  n7 = ScalarParameter() {parameterName=BaseColorRotation}
  MaterialExpressionTextureCoordinate_0 = TextureCoordinate()
  MaterialExpressionTextureCoordinate_1 = TextureCoordinate() {coordinateIndex=1}
  MaterialExpressionTextureCoordinate_2 = TextureCoordinate() {coordinateIndex=2}
  MaterialExpressionTextureCoordinate_3 = TextureCoordinate() {coordinateIndex=3}
  MaterialExpressionScalarParameter_18 = ScalarParameter() {parameterName=BaseColorUVIndex}
  n8 = Custom!(
   UV0=@MaterialExpressionTextureCoordinate_0
   UV1=@MaterialExpressionTextureCoordinate_1
   UV2=@MaterialExpressionTextureCoordinate_2
   UV3=@MaterialExpressionTextureCoordinate_3
   UVIndex=@MaterialExpressionScalarParameter_18
  ) {
   code="switch(UVIndex)\r\n{\r\ncase 0: return UV0;\r\ncase 1: return UV1;\r\ncase 2: return UV2;\r\ncase 3: return UV3;\r\ndefault: return UV0;\r\n}"
   outputType=CMOT_Float2
   inputs [
    {inputName=UV0, input {expression=&self:MaterialExpressionTextureCoordinate_0}}
    {inputName=UV1, input {expression=&self:MaterialExpressionTextureCoordinate_1}}
    {inputName=UV2, input {expression=&self:MaterialExpressionTextureCoordinate_2}}
    {inputName=UV3, input {expression=&self:MaterialExpressionTextureCoordinate_3}}
    {inputName=UVIndex, input {expression=&self:MaterialExpressionScalarParameter_18}}
   ]
  }
  n9 = MaterialFunctionCall(X=@n5.B, Y=@n5.B) {materialFunction=&MakeFloat2}
  LinearInterpolate(
   A=LinearInterpolate(
    A=VectorParameter() {parameterName=BaseColor, defaultValue=<rgba 0 0 0 1>}.RGB
    B=MaterialFunctionCall(X=@n2.Linear, Y=@n3.Linear, Z=@n4.Linear) {materialFunction=&MakeFloat3}.Result
    Alpha=ScalarParameter() {parameterName=UseVertexColorForBaseColor}
   )
   B=TextureSampleParameter2D(UVs=MaterialFunctionCall!(Scale=@n6.Result, Rotation=@n7, TextureCoordinate=@n8, Translation=@n9.Result) {materialFunction=&/UsdTransform2d}.Result) {parameterName=BaseColorTexture, samplerType=SAMPLERTYPE_VirtualColor, texture=/USDCore/Textures/DefaultColorVT}.RGB
   Alpha=ScalarParameter() {parameterName=UseBaseColorTexture}
  )
  n10 = VectorParameter() {parameterName=MetallicScaleTranslation, defaultValue=<rgba 1 1 0 0>}
  MaterialExpressionScalarParameter_15 = ScalarParameter() {parameterName=MetallicUVIndex}
  n11 = Custom(
   UV0=@MaterialExpressionTextureCoordinate_0
   UV1=@MaterialExpressionTextureCoordinate_1
   UV2=@MaterialExpressionTextureCoordinate_2
   UV3=@MaterialExpressionTextureCoordinate_3
   UVIndex=@MaterialExpressionScalarParameter_15
  ) {
   code="switch(UVIndex)\r\n{\r\ncase 0: return UV0;\r\ncase 1: return UV1;\r\ncase 2: return UV2;\r\ncase 3: return UV3;\r\ndefault: return UV0;\r\n}"
   outputType=CMOT_Float2
   inputs [
    {inputName=UV0, input {expression=&self:MaterialExpressionTextureCoordinate_0}}
    {inputName=UV1, input {expression=&self:MaterialExpressionTextureCoordinate_1}}
    {inputName=UV2, input {expression=&self:MaterialExpressionTextureCoordinate_2}}
    {inputName=UV3, input {expression=&self:MaterialExpressionTextureCoordinate_3}}
    {inputName=UVIndex, input {expression=&self:MaterialExpressionScalarParameter_15}}
   ]
  }
  n12 = MaterialFunctionCall(
   Scale=MaterialFunctionCall(X=@n10.R, Y=@n10.R) {materialFunction=&MakeFloat2}.Result
   Rotation=ScalarParameter() {parameterName=MetallicRotation}
   TextureCoordinate=@n11
   Translation=MaterialFunctionCall(X=@n10.B, Y=@n10.B) {materialFunction=&MakeFloat2}.Result
  ) {materialFunction=&UsdTransform2d}
  n13 = VectorParameter() {parameterName=MetallicTextureComponent, defaultValue=<rgba 1 0 0 0>}
  LinearInterpolate(
   A=ScalarParameter() {parameterName=Metallic, sliderMax=1}
   B=MaterialFunctionCall!(
    Value=TextureSampleParameter2D(UVs=@n12.Result) {parameterName=MetallicTexture, samplerType=SAMPLERTYPE_VirtualLinearColor, texture=/USDCore/Textures/DefaultLinearColorVT}.RGBA
    Channel=MaterialFunctionCall(X=@n13.R, Y=@n13.R, Z=@n13.R, A=@n13.R) {materialFunction=&MakeFloat4}.Result
   ) {materialFunction=&/ChannelSelector}.Result
   Alpha=ScalarParameter() {parameterName=UseMetallicTexture}
  )
  n14 = VectorParameter() {parameterName=RoughnessScaleTranslation, defaultValue=<rgba 1 1 0 0>}
  MaterialExpressionScalarParameter_16 = ScalarParameter() {parameterName=RoughnessUVIndex}
  n15 = Custom(
   UV0=@MaterialExpressionTextureCoordinate_0
   UV1=@MaterialExpressionTextureCoordinate_1
   UV2=@MaterialExpressionTextureCoordinate_2
   UV3=@MaterialExpressionTextureCoordinate_3
   UVIndex=@MaterialExpressionScalarParameter_16
  ) {
   code="switch(UVIndex)\r\n{\r\ncase 0: return UV0;\r\ncase 1: return UV1;\r\ncase 2: return UV2;\r\ncase 3: return UV3;\r\ndefault: return UV0;\r\n}"
   outputType=CMOT_Float2
   inputs [
    {inputName=UV0, input {expression=&self:MaterialExpressionTextureCoordinate_0}}
    {inputName=UV1, input {expression=&self:MaterialExpressionTextureCoordinate_1}}
    {inputName=UV2, input {expression=&self:MaterialExpressionTextureCoordinate_2}}
    {inputName=UV3, input {expression=&self:MaterialExpressionTextureCoordinate_3}}
    {inputName=UVIndex, input {expression=&self:MaterialExpressionScalarParameter_16}}
   ]
  }
  n16 = MaterialFunctionCall(
   Scale=MaterialFunctionCall(X=@n14.R, Y=@n14.R) {materialFunction=&MakeFloat2}.Result
   Rotation=ScalarParameter() {parameterName=RoughnessRotation}
   TextureCoordinate=@n15
   Translation=MaterialFunctionCall(X=@n14.B, Y=@n14.B) {materialFunction=&MakeFloat2}.Result
  ) {materialFunction=&UsdTransform2d}
  n17 = VectorParameter() {parameterName=RoughnessTextureComponent, defaultValue=<rgba 1 0 0 0>}
  LinearInterpolate(
   A=ScalarParameter() {defaultValue=1, parameterName=Roughness, sliderMax=1}
   B=MaterialFunctionCall(
    Value=TextureSampleParameter2D(UVs=@n16.Result) {parameterName=RoughnessTexture, samplerType=SAMPLERTYPE_VirtualLinearColor, texture=&DefaultLinearColorVT}.RGBA
    Channel=MaterialFunctionCall(X=@n17.R, Y=@n17.R, Z=@n17.R, A=@n17.R) {materialFunction=&MakeFloat4}.Result
   ) {materialFunction=&ChannelSelector}.Result
   Alpha=ScalarParameter() {parameterName=UseRoughnessTexture}
  )
  n18 = VectorParameter() {parameterName=OpacityScaleTranslation, defaultValue=<rgba 1 1 0 0>}
  MaterialExpressionScalarParameter_19 = ScalarParameter() {parameterName=OpacityUVIndex}
  n19 = Custom(
   UV0=@MaterialExpressionTextureCoordinate_0
   UV1=@MaterialExpressionTextureCoordinate_1
   UV2=@MaterialExpressionTextureCoordinate_2
   UV3=@MaterialExpressionTextureCoordinate_3
   UVIndex=@MaterialExpressionScalarParameter_19
  ) {
   code="switch(UVIndex)\r\n{\r\ncase 0: return UV0;\r\ncase 1: return UV1;\r\ncase 2: return UV2;\r\ncase 3: return UV3;\r\ndefault: return UV0;\r\n}"
   outputType=CMOT_Float2
   inputs [
    {inputName=UV0, input {expression=&self:MaterialExpressionTextureCoordinate_0}}
    {inputName=UV1, input {expression=&self:MaterialExpressionTextureCoordinate_1}}
    {inputName=UV2, input {expression=&self:MaterialExpressionTextureCoordinate_2}}
    {inputName=UV3, input {expression=&self:MaterialExpressionTextureCoordinate_3}}
    {inputName=UVIndex, input {expression=&self:MaterialExpressionScalarParameter_19}}
   ]
  }
  n20 = MaterialFunctionCall(
   Scale=MaterialFunctionCall(X=@n18.R, Y=@n18.R) {materialFunction=&MakeFloat2}.Result
   Rotation=ScalarParameter() {parameterName=OpacityRotation}
   TextureCoordinate=@n19
   Translation=MaterialFunctionCall(X=@n18.B, Y=@n18.B) {materialFunction=&MakeFloat2}.Result
  ) {materialFunction=&UsdTransform2d}
  n21 = VectorParameter() {parameterName=OpacityTextureComponent, defaultValue=<rgba 1 0 0 0>}
  LinearInterpolate(
   A=ScalarParameter() {defaultValue=1, parameterName=Opacity, sliderMax=1}
   B=MaterialFunctionCall(
    Value=TextureSampleParameter2D(UVs=@n20.Result) {parameterName=OpacityTexture, samplerType=SAMPLERTYPE_VirtualLinearColor, texture=&DefaultLinearColorVT}.RGBA
    Channel=MaterialFunctionCall(X=@n21.R, Y=@n21.R, Z=@n21.R, A=@n21.R) {materialFunction=&MakeFloat4}.Result
   ) {materialFunction=&ChannelSelector}.Result
   Alpha=ScalarParameter() {parameterName=UseOpacityTexture}
  )
  n22 = VectorParameter() {parameterName=EmissiveScaleTranslation, defaultValue=<rgba 1 1 0 0>}
  n23 = MaterialFunctionCall(X=@n22.R, Y=@n22.R) {materialFunction=&MakeFloat2}
  n24 = ScalarParameter() {parameterName=EmissiveRotation}
  MaterialExpressionScalarParameter_17 = ScalarParameter() {parameterName=EmissiveUVIndex}
  n25 = Custom(
   UV0=@MaterialExpressionTextureCoordinate_0
   UV1=@MaterialExpressionTextureCoordinate_1
   UV2=@MaterialExpressionTextureCoordinate_2
   UV3=@MaterialExpressionTextureCoordinate_3
   UVIndex=@MaterialExpressionScalarParameter_17
  ) {
   code="switch(UVIndex)\r\n{\r\ncase 0: return UV0;\r\ncase 1: return UV1;\r\ncase 2: return UV2;\r\ncase 3: return UV3;\r\ndefault: return UV0;\r\n}"
   outputType=CMOT_Float2
   inputs [
    {inputName=UV0, input {expression=&self:MaterialExpressionTextureCoordinate_0}}
    {inputName=UV1, input {expression=&self:MaterialExpressionTextureCoordinate_1}}
    {inputName=UV2, input {expression=&self:MaterialExpressionTextureCoordinate_2}}
    {inputName=UV3, input {expression=&self:MaterialExpressionTextureCoordinate_3}}
    {inputName=UVIndex, input {expression=&self:MaterialExpressionScalarParameter_17}}
   ]
  }
  n26 = MaterialFunctionCall(X=@n22.B, Y=@n22.B) {materialFunction=&MakeFloat2}
  LinearInterpolate(
   A=VectorParameter() {parameterName=EmissiveColor, defaultValue=<rgba 0 0 0 1>}.RGB
   B=TextureSampleParameter2D(UVs=MaterialFunctionCall(Scale=@n23.Result, Rotation=@n24, TextureCoordinate=@n25, Translation=@n26.Result) {materialFunction=&UsdTransform2d}.Result) {parameterName=EmissiveColorTexture, samplerType=SAMPLERTYPE_VirtualColor, texture=&DefaultColorVT}.RGB
   Alpha=ScalarParameter() {parameterName=UseEmissiveColorTexture}
  )
  n27 = VectorParameter() {parameterName=NormalScaleTranslation, defaultValue=<rgba 1 1 0 0>}
  n28 = MaterialFunctionCall(X=@n27.R, Y=@n27.R) {materialFunction=&MakeFloat2}
  n29 = ScalarParameter() {parameterName=NormalRotation}
  MaterialExpressionScalarParameter_20 = ScalarParameter() {parameterName=NormalUVIndex}
  n30 = Custom(
   UV0=@MaterialExpressionTextureCoordinate_0
   UV1=@MaterialExpressionTextureCoordinate_1
   UV2=@MaterialExpressionTextureCoordinate_2
   UV3=@MaterialExpressionTextureCoordinate_3
   UVIndex=@MaterialExpressionScalarParameter_20
  ) {
   code="switch(UVIndex)\r\n{\r\ncase 0: return UV0;\r\ncase 1: return UV1;\r\ncase 2: return UV2;\r\ncase 3: return UV3;\r\ndefault: return UV0;\r\n}"
   outputType=CMOT_Float2
   inputs [
    {inputName=UV0, input {expression=&self:MaterialExpressionTextureCoordinate_0}}
    {inputName=UV1, input {expression=&self:MaterialExpressionTextureCoordinate_1}}
    {inputName=UV2, input {expression=&self:MaterialExpressionTextureCoordinate_2}}
    {inputName=UV3, input {expression=&self:MaterialExpressionTextureCoordinate_3}}
    {inputName=UVIndex, input {expression=&self:MaterialExpressionScalarParameter_20}}
   ]
  }
  n31 = MaterialFunctionCall(X=@n27.B, Y=@n27.B) {materialFunction=&MakeFloat2}
  n32 = LinearInterpolate(
   A=VectorParameter() {parameterName=Normal, defaultValue=<rgba 0 0 1 1>}.RGB
   B=TextureSampleParameter2D(UVs=MaterialFunctionCall(Scale=@n28.Result, Rotation=@n29, TextureCoordinate=@n30, Translation=@n31.Result) {materialFunction=&UsdTransform2d}.Result) {parameterName=NormalTexture, samplerType=SAMPLERTYPE_VirtualNormal, texture=/Engine/EngineMaterials/BaseFlattenNormalMap_VT}.RGB
   Alpha=ScalarParameter() {parameterName=UseNormalTexture}
  )
  LinearInterpolate(
   A=@n32
   B=MaterialFunctionCall!(In=@n32) {materialFunction=&/WorldSpaceNormalsToTangentSpace}.Result
   Alpha=ScalarParameter() {desc=Not a static switch so that we can do this at runtime, parameterName=UseWorldSpaceNormals}
  )
  n33 = VectorParameter() {parameterName=RefractionScaleTranslation, defaultValue=<rgba 1 1 0 0>}
  MaterialExpressionScalarParameter_21 = ScalarParameter() {parameterName=RefractionUVIndex}
  n34 = Custom(
   UV0=@MaterialExpressionTextureCoordinate_0
   UV1=@MaterialExpressionTextureCoordinate_1
   UV2=@MaterialExpressionTextureCoordinate_2
   UV3=@MaterialExpressionTextureCoordinate_3
   UVIndex=@MaterialExpressionScalarParameter_21
  ) {
   code="switch(UVIndex)\r\n{\r\ncase 0: return UV0;\r\ncase 1: return UV1;\r\ncase 2: return UV2;\r\ncase 3: return UV3;\r\ndefault: return UV0;\r\n}"
   outputType=CMOT_Float2
   inputs [
    {inputName=UV0, input {expression=&self:MaterialExpressionTextureCoordinate_0}}
    {inputName=UV1, input {expression=&self:MaterialExpressionTextureCoordinate_1}}
    {inputName=UV2, input {expression=&self:MaterialExpressionTextureCoordinate_2}}
    {inputName=UV3, input {expression=&self:MaterialExpressionTextureCoordinate_3}}
    {inputName=UVIndex, input {expression=&self:MaterialExpressionScalarParameter_21}}
   ]
  }
  n35 = MaterialFunctionCall(
   Scale=MaterialFunctionCall(X=@n33.R, Y=@n33.R) {materialFunction=&MakeFloat2}.Result
   Rotation=ScalarParameter() {parameterName=RefractionRotation}
   TextureCoordinate=@n34
   Translation=MaterialFunctionCall(X=@n33.B, Y=@n33.B) {materialFunction=&MakeFloat2}.Result
  ) {materialFunction=&UsdTransform2d}
  n36 = VectorParameter() {parameterName=RefractionTextureComponent, defaultValue=<rgba 1 0 0 0>}
  LinearInterpolate(
   A=ScalarParameter() {defaultValue=1.5, parameterName=Refraction, sliderMax=1}
   B=MaterialFunctionCall(
    Value=TextureSampleParameter2D(UVs=@n35.Result) {parameterName=RefractionTexture, samplerType=SAMPLERTYPE_VirtualLinearColor, texture=&DefaultLinearColorVT}.RGBA
    Channel=MaterialFunctionCall(X=@n36.R, Y=@n36.R, Z=@n36.R, A=@n36.R) {materialFunction=&MakeFloat4}.Result
   ) {materialFunction=&ChannelSelector}.Result
   Alpha=ScalarParameter() {parameterName=UseRefractionTexture}
  )
  n37 = VectorParameter() {parameterName=AmbientOcclusionScaleTranslation, defaultValue=<rgba 1 1 0 0>}
  MaterialExpressionScalarParameter_22 = ScalarParameter() {parameterName=AmbientOcclusionUVIndex}
  n38 = Custom(
   UV0=@MaterialExpressionTextureCoordinate_0
   UV1=@MaterialExpressionTextureCoordinate_1
   UV2=@MaterialExpressionTextureCoordinate_2
   UV3=@MaterialExpressionTextureCoordinate_3
   UVIndex=@MaterialExpressionScalarParameter_22
  ) {
   code="switch(UVIndex)\r\n{\r\ncase 0: return UV0;\r\ncase 1: return UV1;\r\ncase 2: return UV2;\r\ncase 3: return UV3;\r\ndefault: return UV0;\r\n}"
   outputType=CMOT_Float2
   inputs [
    {inputName=UV0, input {expression=&self:MaterialExpressionTextureCoordinate_0}}
    {inputName=UV1, input {expression=&self:MaterialExpressionTextureCoordinate_1}}
    {inputName=UV2, input {expression=&self:MaterialExpressionTextureCoordinate_2}}
    {inputName=UV3, input {expression=&self:MaterialExpressionTextureCoordinate_3}}
    {inputName=UVIndex, input {expression=&self:MaterialExpressionScalarParameter_22}}
   ]
  }
  n39 = TextureSampleParameter2D(UVs=MaterialFunctionCall(
    Scale=MaterialFunctionCall(X=@n37.R, Y=@n37.R) {materialFunction=&MakeFloat2}.Result
    Rotation=ScalarParameter() {parameterName=AmbientOcclusionRotation}
    TextureCoordinate=@n38
    Translation=MaterialFunctionCall(X=@n37.B, Y=@n37.B) {materialFunction=&MakeFloat2}.Result
   ) {materialFunction=&UsdTransform2d}.Result) {parameterName=AmbientOcclusionTexture, samplerType=SAMPLERTYPE_VirtualLinearColor, texture=&DefaultLinearColorVT}
  n40 = VectorParameter() {parameterName=AmbientOcclusionTextureComponent, defaultValue=<rgba 1 0 0 0>}
  LinearInterpolate(
   A=ScalarParameter() {defaultValue=1, parameterName=AmbientOcclusion, sliderMax=1}
   B=MaterialFunctionCall(
    Value=@n39.RGBA
    Channel=MaterialFunctionCall(X=@n40.R, Y=@n40.R, Z=@n40.R, A=@n40.R) {materialFunction=&MakeFloat4}.Result
   ) {materialFunction=&ChannelSelector}.Result
   Alpha=ScalarParameter() {parameterName=UseAmbientOcclusionTexture}
  )
 ]
}
